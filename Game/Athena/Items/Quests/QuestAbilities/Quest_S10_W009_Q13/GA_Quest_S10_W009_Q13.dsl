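; Gameplay ability for the S10 Week 9 Q13 quest that counts headshots dealt with scoped weapons.
; It responds only when the instigator tags match the scoped-weapon tags, the target is a player pawn, and the effect context is a critical hit.
; On activation it sends the quest stat event, then waits briefly before ending as a workaround for damage events firing twice on combined shield and health hits.
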(event Ability|EventActivateAbilityFromEvent
  (params (EventData :type Gameplay%20Event%20Data%20Structure%20%28by%20ref%29))
  (|CallParentFunction EventData)
  (Quest|SendComplexCustomStatEvent (Variables|GAVars|Data|GetOwningPlayerQuestManager) 0 "(GameplayTags=((TagName=\"Athena.Quests.HitHeadshotsScopedWeapons\")))" "(GameplayTags=)" 1)
  ;[1 1 1] HACK: Event.Damage.Dealt fires twice is target takes shield and health damage at the same time. This is a workaround.
  (Utilities|FlowControl|Delay 0.03)
  (Ability|EndAbility))

(fn ShouldAbilityRespondToEvent
  (params (ActorInfo :type Gameplay%20Ability%20Actor%20Info%20Structure) (Payload_EventTag :type Gameplay%20Tag%20Structure) (Payload_Instigator :type Actor%20Object%20Reference) (Payload_Target :type Actor%20Object%20Reference) (Payload_OptionalObject :type Object%20Reference) (Payload_OptionalObject2 :type Object%20Reference) (Payload_ContextHandle :type Gameplay%20Effect%20Context%20Handle%20Structure) (Payload_InstigatorTags :type Gameplay%20Tag%20Container%20Structure) (Payload_TargetTags :type Gameplay%20Tag%20Container%20Structure) (Payload_EventMagnitude :type Float%20%28single-precision%29) (Payload_TargetData :type Gameplay%20Ability%20Target%20Data%20Handle%20Structure) (Payload_InstancedEventData :type Instanced%20Struct%20Structure))
  (if (GameplayTags|TagContainer|CompareTagContainertoOtherTags (Variables|Default|GetT_WeaponScoped) Payload_InstigatorTags)
    (Utilities|Casting|CastToFortPlayerPawn Payload_Target
      (:then
        (return (Ability|EffectContext|IsCriticalHit Payload_ContextHandle)))
      (:CastFailed
        _))
    (else
      _))
  (return false))
